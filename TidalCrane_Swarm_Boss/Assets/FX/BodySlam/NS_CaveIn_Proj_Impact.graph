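# Niagara impact burst for a cave-in projectile, likely part of a boss body-slam attack, throwing out dirt, smoke, streaks and rock debris.
# User.Radius sizes the smoke torus, debris cylinders and streak ring; User.EndPoint is declared but unreferenced here.
# Mesh debris comes as a colliding GPU emitter and a lower-count CPU twin without collision.
# Dirt and streaks tint by a time-of-day parameter and use distance culling.
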NiagaraSystem /TidalCrane_Swarm_Boss/Assets/FX/BodySlam/NS_CaveIn_Proj_Impact
dependencies {
 usedDataInterfaces [
  &NiagaraDataInterfaceCurve
  &NiagaraDataInterfaceArrayFloat3
  &NiagaraDataInterfaceColorCurve
  &NiagaraDataInterfaceMeshRendererInfo
  &NiagaraDataInterfaceVector2DCurve
 ]
 usedDynamicInputs [&SelectVectorFromArray]
 usedModules [
  &SystemState
  &self:SystemSpawnScript.SystemScriptSource.SystemScriptGraph.NiagaraNodeAssignment_0.SetVariables_#1
 ]
}
emitters {
 Dirt {
  emitterUpdateScript [
   &EmitterState(
    Enable Distance Culling=true
    MaxDistance=7500
    Module.bApplySpawnCountScaleByDistance=true
    Module.bUseMaxDistance=true
    Module.bUseMinDistance=false
    Scalability Mode=::NewEnumerator1
    Scale Spawn Count=true
    Spawn Count Scale By Distance=DataInterface() {propertyValues {Curve {keys [<at 0 1>,<at 0.5 1>,<at 1 0.1>]}}}
   )
   &SpawnBurst_Instantaneous(Spawn Count=35, Use Spawn Probability=false, UseLoopCountLimit=false)
  ]
  particleSpawnScript [
   &InitializeParticle(
    Color Minimum=<rgba 1 1 1 0.6>
    Color Mode=::NewEnumerator2
    Lifetime Max=1.5
    Lifetime Mode=::NewEnumerator1
    Sprite Rotation Mode=::NewEnumerator2
    Sprite Size Mode=::NewEnumerator1
    Uniform Sprite Size Max=225
    Uniform Sprite Size Min=150
    UsePositionOffset=false
    Write Material Random=true
   )
   &ShapeLocation(
    Disc Coverage=1
    MaskSpawnGroup=false
    Override Seed=false
    Ring Radius=150
    Rotation Angle=0.3
    Rotation Mode=::NewEnumerator1
    Shape Primitive=::NewEnumerator4
    UseOwnerScale=true
   )
   &AddVelocityFromPoint(
    Module.Use Falloff Distance=false
    Origin Offset=<-75 0 0>
    Velocity Strength=&RandomRangeFloat(Minimum=750, Maximum=1550)
   )
   &AddVelocity(Velocity=&RandomRangeVector(Minimum=<550 0 0>, Maximum=<1200 0 0>))
  ]
  particleUpdateScript [
   &ParticleState()
   &GravityForce(Gravity=&RandomRangeVector(Minimum=<0 0 -1750>, Maximum=<0 0 -3500>))
   &Drag(Drag=&RandomRangeFloat(Minimum=2, Maximum=3), Use Linear Drag=true, Use Rotational Drag=true)
   &ScaleColor(
    Scale Alpha=&FloatFromCurve(FloatCurve=DataInterface() {propertyValues {Curve {keys [<at 0.3 1 RCIM_Cubic>, {arriveTangent=-4.7596, interpMode=RCIM_Cubic, leaveTangent=-4.7596, tangentMode=RCTM_User, time=1, value=-0.1}]}}})
    Scale RGB=&MakeVectorFromLinearColorRGB(Color=&ColorFromCurve(
      ColorCurve=DataInterface() {propertyValues {
        AlphaCurve {keys [<at 0 1 RCIM_Cubic>]}
        BlueCurve {keys [<at 0 0.2812 RCIM_Cubic>,<at 0.25 0.4457>,<at 0.5 0.237>,<at 0.75 0.2751>,<at 1 0.2812 RCIM_Cubic>]}
        GreenCurve {keys [<at 0 0.2256 RCIM_Cubic>,<at 0.25 0.4794>,<at 0.5 0.2543>,<at 0.75 0.3018>,<at 1 0.2143 RCIM_Cubic>]}
        RedCurve {keys [<at 0 0.2045 RCIM_Cubic>,<at 0.25 0.5521>,<at 0.5 0.2917>,<at 0.75 0.3594>,<at 1 0.1834 RCIM_Cubic>]}
       }}
      CurveIndex=Linked() {linkedVariable {name=NPC.FortniteNPC.TimeOfDay_0-1, type=NiagaraFloat}}
     ))
    ScaleA=true
    ScaleRGB=true
   )
   &SolveForcesAndVelocity(Clamp Velocity=false, Limit Acceleration=false)
   &ScaleSpriteSize(Scale Factor=&Vector2DFromFloat(Value=&FloatFromCurve(
      FloatCurve=DataInterface() {propertyValues {Curve {keys [
          {interpMode=RCIM_Cubic, leaveTangent=4.8701, tangentMode=RCTM_Break, time=0, value=1}
          {interpMode=RCIM_Cubic, tangentMode=RCTM_Break, time=1, value=2.5}
         ]}}}
      Scale Curve=&FN_LOD_Distance_Remap(Start Distance=350, End Distance=3500, Minimum Output=0.5, Maximum Output=1, Clamp Output=true)
     )))
   &SpriteRotationRate(Rotation Rate=45)
  ]
  rendererClasses [&NiagaraSpriteRendererProperties]
  renderers [{rendererClass=&NiagaraSpriteRendererProperties}]
 }
 Mesh_Debris01 {
  simTarget=GPUComputeSim
  emitterUpdateScript [
   &EmitterState()
   &SpawnBurst_Instantaneous(Spawn Count=25, Use Spawn Probability=false, UseLoopCountLimit=false)
  ]
  particleSpawnScript [
   &InitializeParticle(
    Color=&RandomRangeLinearColor(Minimum=<rgba 0.4895 0.488 0.526 1>, Maximum=<rgba 0.3426 0.3654 0.401 1>)
    Lifetime Min=1.5
    Lifetime Mode=::NewEnumerator1
    Mass Mode=::NewEnumerator3
    Mesh Scale=&RandomRangeVector(Minimum=<0.055 0.075 0.05>, Maximum=<0.35 0.5 0.35>)
    Mesh Scale Mode=::NewEnumerator4
    Sprite Rotation Mode=::NewEnumerator2
    UsePositionOffset=false
    Write Material Random=true
   )
   &ShapeLocation(
    Cylinder Height=250
    Cylinder Radius=&Multiply_Float(A=$User.Radius, B=0.5)
    MaskSpawnGroup=false
    Override Seed=false
    Rotation Mode=::NewEnumerator1
    Shape Origin=&AddVectorToPosition(Position=$Particles.Position, Vector=<0 0 150>)
    Shape Primitive=::NewEnumerator1
    Surface Only=false
    UseOwnerScale=true
   )
   &AddVelocity(
    Module.Use Falloff Distance=false
    Origin Offset=<0 0 -25>
    Override Seed=false
    Velocity Mode=::NewEnumerator1
    Velocity Speed=&RandomRangeFloat(Minimum=750, Maximum=1750)
   ) {module=AddVelocity001}
   &InitialMeshOrientation(Use Rotation Vector=false)
  ]
  particleUpdateScript [
   &ParticleState()
   &ScaleMeshSize(Scale Factor=&Multiply_VectorByFloat(
     Vector=<1 1 1>
     Float=&FloatFromCurve(
      FloatCurve=DataInterface() {propertyValues {Curve {keys [<at 0.5 1 RCIM_Cubic>, {arriveTangent=-6.8632, interpMode=RCIM_Cubic, leaveTangent=-6.8632, tangentMode=RCTM_User, time=1, value=0}]}}}
      Scale Curve=&RandomRangeFloat(Minimum=0.75, Maximum=2)
     )
    ))
   &GravityForce(Gravity=<0 0 -1550>)
   &Drag(
    Drag=&FloatFromCurve(
     FloatCurve=DataInterface() {propertyValues {Curve {keys [<at 0.1 0.1>, {arriveTangent=-3.6693, interpMode=RCIM_Cubic, leaveTangent=-3.6693, tangentMode=RCTM_User, time=0.1283, value=1}, <at 1 0>]}}}
     Scale Curve=3
    )
    Use Linear Drag=true
    Use Rotational Drag=false
   )
   &Collision(
    Age Colliding Particles=false
    Friction:NiagaraFloat=&RandomRangeFloat(Minimum=0.2, Maximum=0.5)
    GPU Collision Type=::NewEnumerator2
    KillOnCollision=false
    Module.Control Rest State Via Surface Normal=false
    Module.EnableMaxCollisionCount=false
    Module.Randomize Collision Normal Vector=true
    Particle Radius Scale=0.5
    Restitution=0.1
   )
   &SolveForcesAndVelocity(Clamp Velocity=false, Limit Acceleration=false)
  ]
  rendererClasses [&NiagaraMeshRendererProperties]
  renderers [{rendererClass=&NiagaraMeshRendererProperties, rendererIndex=0}]
 }
 Mesh_DebrisLow {
  emitterUpdateScript [
   &EmitterState()
   &SpawnBurst_Instantaneous(Spawn Count=12, Use Spawn Probability=false, UseLoopCountLimit=false)
  ]
  particleSpawnScript [
   &InitializeParticle(
    Color=&RandomRangeLinearColor(Minimum=<rgba 0.4895 0.488 0.526 1>, Maximum=<rgba 0.3426 0.3654 0.401 1>)
    Lifetime Min=1.5
    Lifetime Mode=::NewEnumerator1
    Mass Mode=::NewEnumerator3
    Mesh Scale=&RandomRangeVector(Minimum=<0.055 0.075 0.05>, Maximum=<0.35 0.5 0.35>)
    Mesh Scale Mode=::NewEnumerator4
    Sprite Rotation Mode=::NewEnumerator2
    UsePositionOffset=false
    Write Material Random=true
   )
   &ShapeLocation(
    Cylinder Height=250
    Cylinder Radius=&Multiply_Float(A=$User.Radius, B=0.5)
    MaskSpawnGroup=false
    Override Seed=false
    Rotation Mode=::NewEnumerator1
    Shape Origin=&AddVectorToPosition(Position=$Particles.Position, Vector=<0 0 150>)
    Shape Primitive=::NewEnumerator1
    Surface Only=false
    UseOwnerScale=true
   )
   &AddVelocity(
    Module.Use Falloff Distance=false
    Origin Offset=<0 0 -25>
    Origin Offset Coordinate Space=::World
    Override Seed=false
    Velocity Mode=::NewEnumerator1
    Velocity Speed=&RandomRangeFloat(Minimum=750, Maximum=1750)
   ) {module=AddVelocity001}
   &InitialMeshOrientation(Use Rotation Vector=false)
  ]
  particleUpdateScript [
   &ParticleState()
   &ScaleMeshSize(Scale Factor=&Multiply_VectorByFloat(
     Vector=<1 1 1>
     Float=&FloatFromCurve(
      FloatCurve=DataInterface() {propertyValues {Curve {keys [<at 0.5 1 RCIM_Cubic>, {arriveTangent=-6.8632, interpMode=RCIM_Cubic, leaveTangent=-6.8632, tangentMode=RCTM_User, time=1, value=0}]}}}
      Scale Curve=&RandomRangeFloat(Minimum=0.75, Maximum=2)
     )
    ))
   &GravityForce(Gravity=<0 0 -1550>)
   &Drag(
    Drag=&FloatFromCurve(
     FloatCurve=DataInterface() {propertyValues {Curve {keys [<at 0.1 0.1>, {arriveTangent=-3.6693, interpMode=RCIM_Cubic, leaveTangent=-3.6693, tangentMode=RCTM_User, time=0.1283, value=1}, <at 1 0>]}}}
     Scale Curve=3
    )
    Use Linear Drag=true
    Use Rotational Drag=false
   )
   &SolveForcesAndVelocity(Clamp Velocity=false, Limit Acceleration=false)
  ]
  rendererClasses [&NiagaraMeshRendererProperties]
  renderers [{rendererClass=&NiagaraMeshRendererProperties, rendererIndex=0}]
 }
 Smoke01 {
  emitterUpdateScript [
   &EmitterState(Life Cycle Mode=::NewEnumerator1, Loop Behavior=::NewEnumerator1, Loop Delay=0.5, Loop Duration=2, UseLoopDelay=false)
   &SpawnBurst_Instantaneous(
    Spawn Count=&RandomRangeInt(Minimum=12, Maximum=15)
    Use Spawn Probability=false
    UseLoopCountLimit=false
   )
   &SpawnBurst_Instantaneous(
    Spawn Count=&RandomRangeInt(Minimum=5, Maximum=8)
    Spawn Group=1
    Use Spawn Probability=false
    UseLoopCountLimit=false
   ) {module=SpawnBurst_Instantaneous001}
  ]
  particleSpawnScript [
   &InitializeParticle(
    Lifetime Max=1.25
    Lifetime Min=0.65
    Lifetime Mode=::NewEnumerator1
    Mass Mode=::NewEnumerator3
    Sprite Rotation Mode=::NewEnumerator2
    Sprite Size Mode=::NewEnumerator1
    Uniform Sprite Size Max=750
    Uniform Sprite Size Min=400
    UsePositionOffset=false
    Write Material Random=true
   )
   &ShapeLocation(
    Large Radius=&MaskFloatBySpawnGroup(Masked Float=100, Passthrough Float=$User.Radius, Spawn Group Mask=1)
    MaskSpawnGroup=false
    Rotation Mode=::NewEnumerator1
    Shape Primitive=::NewEnumerator3
    Torus Distribution Mode:ENiagaraTorusDistributionMode=::NewEnumerator1
    U Position=&Add_Float(
     A=&ReturnNormalizedExecIndex()
     B=&RandomRangeFloat(Minimum=-0.1, Maximum=0.1)
    )
    UseOwnerScale=true
   ) {module=ShapeLocation001}
   &AddVelocity(
    Module.Use Falloff Distance=false
    Origin Offset=&MaskVectorBySpawnGroup(Masked Vector=<-550 0 0>, Passthrough Vector=<-75 0 0>, Spawn Group Mask=1)
    Override Seed=false
    Velocity Mode=::NewEnumerator1
    Velocity Speed=&RandomRangeFloat(Minimum=1250, Maximum=2500)
   ) {module=AddVelocity001}
  ]
  particleUpdateScript [
   &ParticleState()
   &ScaleSpriteSize(
    Uniform Curve Scale=&FN_LOD_Distance_Remap(Start Distance=350, End Distance=3500, Minimum Output=0.5, Maximum Output=1, Clamp Output=true)
    Uniform Curve Sprite Scale=DataInterface() {propertyValues {Curve {keys [
        {
         arriveTangent=5.9659
         arriveTangentWeight=0.5845
         interpMode=RCIM_Cubic
         leaveTangent=5.9659
         leaveTangentWeight=0.5845
         tangentMode=RCTM_User
         tangentWeightMode=RCTWM_WeightedBoth
         time=0
         value=0.35
        }
        {tangentMode=RCTM_User, time=1, value=1}
       ]}}}
   )
   /SetVariables(Particles.SubImageIndex=&FloatFromCurve(FloatCurve=DataInterface() {propertyValues {Curve {keys [
         {arriveTangent=29.8623, interpMode=RCIM_Cubic, leaveTangent=29.8623, tangentMode=RCTM_User, time=-0.0046, value=5.0556}
         <at 1.0011 22.1689 RCIM_Cubic>
        ]}}})) {isSetParameters=true}
   &DynamicMaterialParameters(
    Index 0 Param 1=&FloatFromCurve(FloatCurve=DataInterface() {propertyValues {Curve {keys [
         {arriveTangent=-0.9727, interpMode=RCIM_Cubic, leaveTangent=-0.9727, tangentMode=RCTM_User, time=-0.0046, value=1.0113}
         {arriveTangent=-1.3074, tangentMode=RCTM_User, time=1.0011, value=0.002}
        ]}}})
    Index 0 Param 4=&FloatFromCurve(FloatCurve=DataInterface() {propertyValues {Curve {keys [<at 0 0 RCIM_Cubic>, {arriveTangent=1.4299, tangentMode=RCTM_User, time=1, value=0.5}]}}})
    Module.Param0WriteEnabled=true
    Module.Param1WriteEnabled=false
    Module.Param2WriteEnabled=false
    Module.Param3WriteEnabled=true
   )
   &Color(Color=&RandomRangeLinearColor(
     Minimum=&ColorFromCurve(ColorCurve=DataInterface() {propertyValues {
        AlphaCurve {keys [<at 0.4505 0.25 RCIM_Cubic>,<at 1 0 RCIM_Cubic>]}
        BlueCurve {keys [<at 0 0.1615 RCIM_Cubic>,<at 1 0.3542 RCIM_Cubic>]}
        GreenCurve {keys [<at 0 0.1527 RCIM_Cubic>,<at 1 0.2374 RCIM_Cubic>]}
        RedCurve {keys [<at 0 0.1161 RCIM_Cubic>,<at 1 0.2431 RCIM_Cubic>]}
       }})
     Maximum=&ColorFromCurve(ColorCurve=DataInterface() {propertyValues {
        AlphaCurve {keys [<at 0.4743 0.35 RCIM_Cubic>,<at 1 0 RCIM_Cubic>]}
        BlueCurve {keys [<at 0 0.0907 RCIM_Cubic>,<at 1 0.1463 RCIM_Cubic>]}
        GreenCurve {keys [<at 0 0.0938 RCIM_Cubic>,<at 1 0.15 RCIM_Cubic>]}
        RedCurve {keys [<at 0 0.0794 RCIM_Cubic>,<at 1 0.1347 RCIM_Cubic>]}
       }})
    ))
   &Drag(
    Drag=&MaskFloatBySpawnGroup(Masked Float=8, Passthrough Float=3)
    Use Linear Drag=true
    Use Rotational Drag=false
   )
   &SolveForcesAndVelocity(Clamp Velocity=false, Limit Acceleration=false)
  ]
  rendererClasses [&NiagaraSpriteRendererProperties]
  renderers [{rendererClass=&NiagaraSpriteRendererProperties}]
 }
 Swirl_Streaks {
  emitterUpdateScript [
   &EmitterState(Enable Distance Culling=true, MaxDistance=7500, Module.bUseMaxDistance=true, Module.bUseMinDistance=false, Scalability Mode=::NewEnumerator1)
   &SpawnBurst_Instantaneous(Spawn Count=7, Use Spawn Probability=false, UseLoopCountLimit=false)
  ]
  particleSpawnScript [
   &InitializeParticle(
    Color=&RandomRangeLinearColor(Minimum=<rgba 0.7332 0.7993 0.9427 1.5>, Maximum=<rgba 0.3476 0.4169 0.5104 2>)
    Lifetime Max=0.85
    Lifetime Min=0.55
    Lifetime Mode=::NewEnumerator1
    Lifetime Randomness Mode:ENiagaraRandomnessMode=::NewEnumerator2
    Sprite Size=&UniformRangedVector2D(Minimum=<120 175>, Maximum=<250 350>)
    Sprite Size Mode=::NewEnumerator4
    UsePositionOffset=false
    Write Material Random=true
   )
   &ShapeLocation(
    MaskSpawnGroup=false
    Ring / Disc Distribution Mode:ENiagaraSphereDistributionMode=::NewEnumerator3
    Ring Radius=&Divide_Float(A=$User.Radius, B=1.35)
    Rotation Angle=0.25
    Rotation Mode=::NewEnumerator1
    Shape Primitive=::NewEnumerator4
    UseOwnerScale=true
   )
   &AddVelocity(Module.Use Falloff Distance=false, Origin Offset=<-200 0 0>, Override Seed=false, Velocity Mode=::NewEnumerator1, Velocity Speed=50)
   &SubUVAnimation(End Frame=3, SubUV Animation Mode:ENiagara_SubUVLookupMode=::NewEnumerator2)
   &CameraOffset(Camera Offset Amount=75)
  ]
  particleUpdateScript [
   &ParticleState()
   &Drag(Use Linear Drag=true, Use Rotational Drag=true)
   &SolveForcesAndVelocity(Clamp Velocity=false, Limit Acceleration=false)
   &DynamicMaterialParameters(
    Index 0 Param 1=25
    Index 0 Param 4=&FloatFromCurve(FloatCurve=DataInterface() {propertyValues {Curve {keys [{interpMode=RCIM_Cubic, tangentMode=RCTM_User, time=0, value=0}, {interpMode=RCIM_Cubic, tangentMode=RCTM_User, time=1, value=1}]}}})
    Index 1 Param 1=&FloatFromCurve(
     FloatCurve=DataInterface() {propertyValues {Curve {keys [
         {interpMode=RCIM_Cubic, tangentMode=RCTM_User, time=0, value=0}
         {arriveTangent=0.2407, interpMode=RCIM_Cubic, leaveTangent=0.2407, tangentMode=RCTM_User, time=1, value=1}
        ]}}}
     Scale Curve=&RandomRangeFloat(Minimum=0.1, Maximum=1)
    )
    Index 1 Param 3=0.5
    Index 1 Param 4=&FloatFromCurve(FloatCurve=DataInterface() {propertyValues {Curve {keys [<at 0 1>,<at 1 0>]}}}, Scale Curve=-0.3)
    Module.Index0Param0Write=true
    Module.Index0Param1Write=false
    Module.Index0Param2Write=false
    Module.Index0Param3Write=true
    Module.Index1Param0Write=true
    Module.Index1Param1Write=true
    Module.Index1Param2Write=true
    Module.Index1Param3Write=true
    Write Parameter Index 1=true
   )
   &ScaleSpriteSize(Scale Factor=&Vector2DFromCurve(
     Vector2Curve=DataInterface() {propertyValues {
       XCurve {keys [
         {arriveTangent=0.4813, interpMode=RCIM_Cubic, leaveTangent=0.4813, tangentMode=RCTM_User, time=0, value=0.625}
         {arriveTangent=1.25, interpMode=RCIM_Cubic, leaveTangent=1.25, time=0.4, value=0.9716}
         {arriveTangent=2.6138, interpMode=RCIM_Cubic, leaveTangent=2.6138, tangentMode=RCTM_User, time=0.9, value=1.75}
        ]}
       YCurve {keys [
         {arriveTangent=9.4573, interpMode=RCIM_Cubic, leaveTangent=9.4573, tangentMode=RCTM_User, time=0, value=0}
         {arriveTangent=0.8193, interpMode=RCIM_Cubic, leaveTangent=0.8193, tangentMode=RCTM_User, time=0.2, value=1.5019}
         <at 1 1.7992 RCIM_Cubic>
        ]}
      }}
     ScaleCurve=<0.5 1>
    ))
   &ScaleSpriteSize(
    Scale Factor=&MakeVector2D(
     X=&FloatFromCurve(
      FloatCurve=DataInterface() {propertyValues {Curve {keys [
          {arriveTangent=-0.2333, interpMode=RCIM_Cubic, leaveTangent=-0.2333, tangentMode=RCTM_User, time=0, value=2}
          {arriveTangent=-3.2303, interpMode=RCIM_Cubic, tangentMode=RCTM_Break, time=1, value=1}
         ]}}}
      Scale Curve=0.9
     )
     Y=&FloatFromCurve(FloatCurve=DataInterface() {propertyValues {Curve {keys [{arriveTangent=1.8705, interpMode=RCIM_Cubic, time=0, value=0.75}, {interpMode=RCIM_Cubic, leaveTangent=5.1925, time=0.3, value=2.5}]}}})
    )
    Scale Sprite Size Mode=::NewEnumerator2
   ) {module=ScaleSpriteSize001}
   &ScaleColor(
    Scale Alpha=&FloatFromCurve(FloatCurve=DataInterface() {propertyValues {Curve {keys [
         {arriveTangent=9.0725, interpMode=RCIM_Cubic, leaveTangent=9.0725, tangentMode=RCTM_User, time=0, value=0}
         {arriveTangent=-0.0923, interpMode=RCIM_Cubic, leaveTangent=-0.0923, tangentMode=RCTM_User, time=0.075, value=1}
         {arriveTangent=-3.3957, interpMode=RCIM_Cubic, leaveTangent=-3.3957, tangentMode=RCTM_User, time=1, value=0}
        ]}}})
    Scale RGB=&MakeVectorFromLinearColorRGB(Color=&ColorFromCurve(
      ColorCurve=DataInterface() {propertyValues {
        AlphaCurve {keys [<at 0 1 RCIM_Cubic>]}
        BlueCurve {keys [<at 0.061 0.8594 RCIM_Cubic>,<at 0.25 0.2849>,<at 0.5 0.3086>,<at 0.75 0.3238>,<at 0.8984 0.8594 RCIM_Cubic>]}
        GreenCurve {keys [<at 0.061 0.4819 RCIM_Cubic>,<at 0.25 0.341>,<at 0.5 0.3186>,<at 0.75 0.3425>,<at 0.8984 0.4819 RCIM_Cubic>]}
        RedCurve {keys [<at 0.061 0.3939 RCIM_Cubic>,<at 0.25 0.3594>,<at 0.5 0.3385>,<at 0.75 0.3594>,<at 0.8984 0.3939 RCIM_Cubic>]}
       }}
      CurveIndex=Linked() {linkedVariable {name=NPC.FortniteNPC.TimeOfDay_0-1, type=NiagaraFloat}}
     ))
    ScaleA=true
    ScaleRGB=true
   )
  ]
  rendererClasses [&NiagaraSpriteRendererProperties]
  renderers [{rendererClass=&NiagaraSpriteRendererProperties}]
 }
}
summary {
 systemName=NS_CaveIn_Proj_Impact
 emitters [
  {emitterName=Smoke01, rendererClasses [&NiagaraSpriteRendererProperties]}
  {emitterName=Swirl_Streaks, rendererClasses [&NiagaraSpriteRendererProperties]}
  {emitterName=Dirt, rendererClasses [&NiagaraSpriteRendererProperties]}
  {emitterName=Mesh_Debris01, simTarget=GPUComputeSim, rendererClasses [&NiagaraMeshRendererProperties]}
  {emitterName=Mesh_DebrisLow, rendererClasses [&NiagaraMeshRendererProperties]}
 ]
 userVariables [
  User.EndPoint {type=NiagaraPosition, description="", defaultValue {}}
  User.Radius {type=NiagaraFloat, defaultValue=250, description=""}
 ]
}
