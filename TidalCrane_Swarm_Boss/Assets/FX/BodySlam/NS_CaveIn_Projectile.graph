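# Niagara effect in the BodySlam FX set that draws a tumbling mesh, apparently as the visual for a cave-in projectile.
# Its one emitter, SingleLoopingParticle, burst-spawns a mesh particle that lives for the current loop duration, with randomized uniform scale and initial orientation.
# Each update rotates the mesh around a slightly tilted X axis at a randomized rate.
# It has no velocity modules or user variables, so the system itself supplies no travel motion.
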NiagaraSystem /TidalCrane_Swarm_Boss/Assets/FX/BodySlam/NS_CaveIn_Projectile
dependencies {
 usedDataInterfaces [&NiagaraDataInterfaceMeshRendererInfo]
 usedDynamicInputs []
 usedModules [&SystemState]
}
emitters {SingleLoopingParticle {
  emitterUpdateScript [
   &EmitterState()
   &SpawnBurst_Instantaneous(Use Spawn Probability=false, UseLoopCountLimit=false)
  ]
  particleSpawnScript [
   &InitializeParticle(
    Color Mode=::NewEnumerator0
    Lifetime=$Emitter.CurrentLoopDuration
    Mesh Renderer Array Visibility Mode=::NewEnumerator3
    Mesh Renderer Info=DataInterface() {propertyValues {MeshRenderer=&self:SingleLoopingParticle.NiagaraMeshRendererProperties_0}}
    Mesh Scale Mode=::NewEnumerator1
    Mesh Uniform Scale Max=1.25
    Mesh Uniform Scale Min=0.75
    Sprite Size Mode=::NewEnumerator1
    Uniform Sprite Size Max=1.25
    Uniform Sprite Size Min=0.75
    UsePositionOffset=false
    Write Material Random=true
   )
   &InitialMeshOrientation(
    Mesh Orientation Mode=::NewEnumerator2
    Rotation=&RandomRangeVector(Minimum=<0 -0.1 -0.1>, Maximum=<1 0.1 0.1>)
    Use Rotation Vector=true
   )
  ]
  particleUpdateScript [
   &ParticleState()
   &UpdateMeshOrientation(Rotation Rate=&RandomRangeFloat(Minimum=0.35, Maximum=0.55), Rotation Vector=<1 0.1 -0.1>)
  ]
  rendererClasses [&NiagaraMeshRendererProperties]
  renderers [{rendererClass=&NiagaraMeshRendererProperties, rendererIndex=0}]
 }}
summary {
 systemName=NS_CaveIn_Projectile
 emitters [{emitterName=SingleLoopingParticle, rendererClasses [&NiagaraMeshRendererProperties]}]
 userVariables []
}
