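# Niagara impact effect for a body slam, bursting dirt, GPU colliding debris, smoke, GPU sparks and spikes along a strip sized by User.DamageAreaLength.
# Spawn groups index vector arrays to mirror particles sideways and outward, with an extra upward, high-drag smoke burst.
# A mesh trail grows, fades and lerps toward the damage length.
# User.Time Delay offsets every emitter; the DamageAreaWidth and HUE parameters are unreferenced in the shown modules.
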NiagaraSystem /TidalCrane_Swarm_Boss/Assets/FX/BodySlam/NS_Hit_BodySlam01
dependencies {
 usedDataInterfaces [
  &NiagaraDataInterfaceArrayFloat3
  &NiagaraDataInterfaceColorCurve
  &NiagaraDataInterfaceCurve
  &NiagaraDataInterfaceArrayFloat
  &NiagaraDataInterfaceVector2DCurve
  &NiagaraDataInterfaceVectorCurve
 ]
 usedDynamicInputs []
 usedModules [&SystemState]
}
emitters {
 Dirt01 {
  emitterUpdateScript [
   &EmitterState(Life Cycle Mode=::NewEnumerator1, Loop Behavior=::NewEnumerator1, Loop Delay=$User.Time Delay, Loop Duration=2, UseLoopDelay=true)
   &SpawnBurst_Instantaneous(Spawn Count=100, Use Spawn Probability=false, UseLoopCountLimit=false)
   &SpawnBurst_Instantaneous(Spawn Count=100, Spawn Group=1, Use Spawn Probability=false, UseLoopCountLimit=false) {module=SpawnBurst_Instantaneous001}
  ]
  particleSpawnScript [
   &InitializeParticle(
    Lifetime Max=1
    Lifetime Min=0.2
    Lifetime Mode=::NewEnumerator1
    Mass Mode=::NewEnumerator3
    Sprite Rotation Mode=::NewEnumerator2
    Sprite Size Mode=::NewEnumerator1
    Uniform Sprite Size Max=350
    Uniform Sprite Size Min=250
    UsePositionOffset=false
    Write Material Random=true
   )
   &CameraOffset(Camera Offset Amount=50)
   /SetVariables(Particles.GroupID:NiagaraInt32=$Emitter.SpawnGroup) {isSetParameters=true, module=SetVariables_#1}
   &ShapeLocation(
    Box Midpoint=<0 0.5 0.5>
    Box Size=&MakeVector(X=$User.DamageAreaLength, Y=100, Z=50)
    MaskSpawnGroup=false
    Offset=&SelectVectorFromArray(Direct Array Index=$Particles.GroupID, Vector Selection Array=DataInterface() {propertyValues {FloatData [<0 -200 0>, <0 200 0>]}})
    Offset Coordinate Space=::Local
    Offset Mode=::NewEnumerator0
    Override Seed=false
    Shape Origin=$Engine.Emitter.SimulationPosition
    Shape Primitive=::NewEnumerator2
    Surface Only=false
    UseOwnerScale=true
   )
   &AddVelocity(
    Cone Angle=20
    Cone Axis=&SelectVectorFromArray(Direct Array Index=$Particles.GroupID, Vector Selection Array=DataInterface() {propertyValues {FloatData [<0.5 -1 1>, <0.5 1 1>]}})
    Distribution Along Cone Axis=0
    Module.UseConeFalloff=false
    Module.UseVelDistribution=false
    Override Seed=false
    Speed Falloff From Cone Axis=0
    Velocity Mode=::NewEnumerator2
    Velocity Speed=&RandomRangeFloat(Minimum=750, Maximum=2000)
   ) {module=AddVelocity001}
  ]
  particleUpdateScript [
   &ParticleState()
   &Color(Color=&RandomRangeLinearColor(
     Minimum=&ColorFromCurve(ColorCurve=DataInterface() {propertyValues {
        AlphaCurve {keys [<at 0.4505 1 RCIM_Cubic>,<at 1 0 RCIM_Cubic>]}
        BlueCurve {keys [<at 0 0.2025 RCIM_Cubic>,<at 1 0.2025 RCIM_Cubic>]}
        GreenCurve {keys [<at 0 0.3859 RCIM_Cubic>,<at 1 0.3858 RCIM_Cubic>]}
        RedCurve {keys [<at 0 0.5972 RCIM_Cubic>,<at 1 0.5972 RCIM_Cubic>]}
       }})
     Maximum=&ColorFromCurve(ColorCurve=DataInterface() {propertyValues {
        AlphaCurve {keys [<at 0.4743 1 RCIM_Cubic>,<at 1 0 RCIM_Cubic>]}
        BlueCurve {keys [<at 0 0.0508 RCIM_Cubic>,<at 1 0.0508 RCIM_Cubic>]}
        GreenCurve {keys [<at 0 0.0974 RCIM_Cubic>,<at 1 0.0974 RCIM_Cubic>]}
        RedCurve {keys [<at 0 0.15 RCIM_Cubic>,<at 1 0.15 RCIM_Cubic>]}
       }})
    ))
   &GravityForce(Gravity=&RandomRangeVector(Minimum=<0 0 -1000>, Maximum=<1 1 -2000>))
   &Drag(
    Drag=&FloatFromCurve(
     FloatCurve=DataInterface() {propertyValues {Curve {
        defaultValue=1
        keys [
         {tangentMode=RCTM_User, time=0, value=0.2}
         {arriveTangent=-2.8845, interpMode=RCIM_Cubic, leaveTangent=-2.8845, tangentMode=RCTM_User, time=0.1, value=1}
         {tangentMode=RCTM_User, time=1, value=0.2}
        ]
       }}}
     Scale Curve=2
    )
    Use Linear Drag=true
    Use Rotational Drag=false
   )
   &ScaleColor(
    Scale Alpha=&FloatFromCurve(FloatCurve=DataInterface() {propertyValues {Curve {keys [<at 0.5 1.0113 RCIM_Cubic>,<at 1.0011 0.002>]}}})
    ScaleA=true
    ScaleRGB=false
   ) {enabled=false}
   &SolveForcesAndVelocity(Clamp Velocity=false, Limit Acceleration=false)
  ]
  rendererClasses [&NiagaraSpriteRendererProperties]
  renderers [{rendererClass=&NiagaraSpriteRendererProperties}]
 }
 Mesh_Debris01 {
  simTarget=GPUComputeSim
  emitterUpdateScript [
   &EmitterState(Life Cycle Mode=::NewEnumerator1, Loop Behavior=::NewEnumerator1, Loop Delay=$User.Time Delay, Loop Duration=1, UseLoopDelay=true)
   &SpawnBurst_Instantaneous(
    Spawn Count=&RandomRangeInt(Minimum=50, Maximum=70)
    Use Spawn Probability=false
    UseLoopCountLimit=false
   )
   &SpawnBurst_Instantaneous(
    Spawn Count=&RandomRangeInt(Minimum=50, Maximum=70)
    Spawn Group=1
    Use Spawn Probability=false
    UseLoopCountLimit=false
   ) {module=SpawnBurst_Instantaneous001}
  ]
  particleSpawnScript [
   &InitializeParticle(
    Color=<rgba 0.1788 0.0883 0.0319 1>
    Lifetime Max=1.5
    Lifetime Min=0.5
    Lifetime Mode=::NewEnumerator1
    Mass Mode=::NewEnumerator3
    Mesh Scale=&RandomRangeVector(Minimum=<0.1 0.2 0.1>, Maximum=<0.3 0.5 0.6>)
    Mesh Scale Mode=::NewEnumerator4
    Sprite Rotation Mode=::NewEnumerator2
    Sprite Size Mode=::NewEnumerator1
    Uniform Sprite Size Max=12
    Uniform Sprite Size Min=6
    UsePositionOffset=false
    Write Material Random=true
   )
   /SetVariables(Particles.GroupID:NiagaraInt32=$Emitter.SpawnGroup) {isSetParameters=true, module=SetVariables_#1}
   &ShapeLocation(
    Box Midpoint=<0 0.5 0.5>
    Box Size=&MakeVector(X=$User.DamageAreaLength, Y=100, Z=50)
    MaskSpawnGroup=false
    Offset=&SelectVectorFromArray(Direct Array Index=$Particles.GroupID, Vector Selection Array=DataInterface() {propertyValues {FloatData [<0 -200 0>, <0 200 0>]}})
    Offset Coordinate Space=::Local
    Offset Mode=::NewEnumerator0
    Override Seed=false
    Shape Origin=$Engine.Emitter.SimulationPosition
    Shape Primitive=::NewEnumerator2
    Surface Only=false
    UseOwnerScale=true
   )
   &AddVelocity(
    Cone Angle=20
    Cone Axis=&SelectVectorFromArray(Direct Array Index=$Particles.GroupID, Vector Selection Array=DataInterface() {propertyValues {FloatData [<0.5 -1 1.25>, <0.5 1 1.25>]}})
    Distribution Along Cone Axis=0
    Module.UseConeFalloff=false
    Module.UseVelDistribution=false
    Override Seed=false
    Speed Falloff From Cone Axis=0
    Velocity Mode=::NewEnumerator2
    Velocity Speed=&RandomRangeFloat(Minimum=750, Maximum=3250)
   ) {module=AddVelocity001}
   &InitialMeshOrientation(Use Rotation Vector=false)
  ]
  particleUpdateScript [
   &ParticleState()
   &Color(Color=<rgba 0.9878 0.5348 0.3349 1>)
   &ScaleMeshSize(Scale Factor=&Multiply_VectorByFloat(
     Vector=<1 1 1>
     Float=&FloatFromCurve(FloatCurve=DataInterface() {propertyValues {Curve {keys [<at 0.5 1>,<at 1 0>]}}})
    ))
   &GravityForce(Gravity=&RandomRangeVector(Minimum=<0 0 -1000>, Maximum=<1 1 -2000>))
   &Drag(
    Drag=&FloatFromCurve(
     FloatCurve=DataInterface() {propertyValues {Curve {defaultValue=1, keys [{tangentMode=RCTM_User, time=0, value=0}, <at 0.3 0.5>, {tangentMode=RCTM_User, time=1, value=0}]}}}
     Scale Curve=5
    )
    Use Linear Drag=true
    Use Rotational Drag=false
   )
   &Collision(
    Age Colliding Particles=false
    Friction:NiagaraFloat=&RandomRangeFloat(Minimum=0.2, Maximum=0.5)
    GPU Collision Type=::NewEnumerator2
    KillOnCollision=false
    Module.Control Rest State Via Surface Normal=false
    Module.EnableMaxCollisionCount=false
    Module.Randomize Collision Normal Vector=true
    Particle Radius Scale=0.5
    Restitution=0.1
   )
   &SolveForcesAndVelocity(Clamp Velocity=false, Limit Acceleration=false)
  ]
  rendererClasses [&NiagaraMeshRendererProperties]
  renderers [{rendererClass=&NiagaraMeshRendererProperties, rendererIndex=0}]
 }
 Smoke01 {
  emitterUpdateScript [
   &EmitterState(Life Cycle Mode=::NewEnumerator1, Loop Behavior=::NewEnumerator1, Loop Delay=$User.Time Delay, Loop Duration=2, UseLoopDelay=true)
   &SpawnBurst_Instantaneous(
    Spawn Count=&RandomRangeInt(Minimum=15, Maximum=25)
    Use Spawn Probability=false
    UseLoopCountLimit=false
   )
   &SpawnBurst_Instantaneous(
    Spawn Count=&RandomRangeInt(Minimum=15, Maximum=25)
    Spawn Group=1
    Use Spawn Probability=false
    UseLoopCountLimit=false
   ) {module=SpawnBurst_Instantaneous001}
   &SpawnBurst_Instantaneous(
    Spawn Count=&RandomRangeInt(Minimum=12, Maximum=15)
    Spawn Group=2
    Use Spawn Probability=false
    UseLoopCountLimit=false
   ) {module=SpawnBurst_Instantaneous002}
  ]
  particleSpawnScript [
   &InitializeParticle(
    Lifetime Max=1.25
    Lifetime Min=0.65
    Lifetime Mode=::NewEnumerator1
    Mass Mode=::NewEnumerator3
    Sprite Rotation Mode=::NewEnumerator2
    Sprite Size Mode=::NewEnumerator1
    Uniform Sprite Size Max=650
    Uniform Sprite Size Min=450
    UsePositionOffset=false
    Write Material Random=true
   )
   &CameraOffset(Camera Offset Amount=70)
   /SetVariables(Particles.GroupID:NiagaraInt32=$Emitter.SpawnGroup) {isSetParameters=true, module=SetVariables_#1}
   &ShapeLocation(
    Box Midpoint=<0 0.5 0>
    Box Size=&MakeVector(X=$User.DamageAreaLength, Y=100, Z=50)
    MaskSpawnGroup=false
    Offset=&SelectVectorFromArray(Direct Array Index=$Particles.GroupID, Vector Selection Array=DataInterface() {propertyValues {FloatData [<0 -200 0>, <0 200 0>]}})
    Offset Coordinate Space=::Local
    Offset Mode=::NewEnumerator0
    Override Seed=false
    Shape Origin=$Engine.Emitter.SimulationPosition
    Shape Primitive=::NewEnumerator2
    Surface Only=false
    UseOwnerScale=true
   )
   &AddVelocity(
    Cone Angle=30
    Cone Axis=&SelectVectorFromArray(
     Direct Array Index=$Particles.GroupID
     Vector Selection Array=DataInterface() {propertyValues {FloatData [<0.5 -1 1.25>, <0.5 1 1.25>, <0 0 1>]}}
    )
    Distribution Along Cone Axis=0
    Module.UseConeFalloff=false
    Module.UseVelDistribution=false
    Override Seed=false
    Speed Falloff From Cone Axis=150
    Velocity Mode=::NewEnumerator2
    Velocity Speed=&RandomRangeFloat(Minimum=750, Maximum=2250)
   ) {module=AddVelocity001}
  ]
  particleUpdateScript [
   &ParticleState()
   /SetVariables(Particles.SubImageIndex=&FloatFromCurve(FloatCurve=DataInterface() {propertyValues {Curve {keys [
         {arriveTangent=29.8623, interpMode=RCIM_Cubic, leaveTangent=29.8623, tangentMode=RCTM_User, time=-0.0046, value=5.0556}
         <at 1.0011 22.1689 RCIM_Cubic>
        ]}}})) {isSetParameters=true}
   &DynamicMaterialParameters(
    Index 0 Param 1=&FloatFromCurve(FloatCurve=DataInterface() {propertyValues {Curve {keys [<at -0.0046 1.0113 RCIM_Cubic>,<at 1.0011 0.002>]}}})
    Index 0 Param 4=&FloatFromCurve(FloatCurve=DataInterface() {propertyValues {Curve {keys [<at 0 0 RCIM_Cubic>,<at 1 0.5>]}}})
    Module.Param0WriteEnabled=true
    Module.Param1WriteEnabled=false
    Module.Param2WriteEnabled=false
    Module.Param3WriteEnabled=true
   )
   &Color(Color=&RandomRangeLinearColor(
     Minimum=&ColorFromCurve(ColorCurve=DataInterface() {propertyValues {
        AlphaCurve {keys [<at 0.4505 0.3 RCIM_Cubic>,<at 1 0 RCIM_Cubic>]}
        BlueCurve {keys [<at 0 0.2025 RCIM_Cubic>,<at 1 0.2025 RCIM_Cubic>]}
        GreenCurve {keys [<at 0 0.3859 RCIM_Cubic>,<at 1 0.3858 RCIM_Cubic>]}
        RedCurve {keys [<at 0 0.5972 RCIM_Cubic>,<at 1 0.5972 RCIM_Cubic>]}
       }})
     Maximum=&ColorFromCurve(ColorCurve=DataInterface() {propertyValues {
        AlphaCurve {keys [<at 0.4743 0.15 RCIM_Cubic>,<at 1 0 RCIM_Cubic>]}
        BlueCurve {keys [<at 0 0.0508 RCIM_Cubic>,<at 1 0.0508 RCIM_Cubic>]}
        GreenCurve {keys [<at 0 0.0974 RCIM_Cubic>,<at 1 0.0974 RCIM_Cubic>]}
        RedCurve {keys [<at 0 0.15 RCIM_Cubic>,<at 1 0.15 RCIM_Cubic>]}
       }})
    ))
   &GravityForce(Gravity=&RandomRangeVector(Minimum=<0 0 -1000>, Maximum=<1 1 -2500>))
   &Drag(
    Drag=&FloatFromCurve(
     FloatCurve=DataInterface() {propertyValues {Curve {defaultValue=1, keys [{tangentMode=RCTM_User, time=0, value=0}, <at 0.3 0.5>, {tangentMode=RCTM_User, time=1, value=0}]}}}
     Scale Curve=&SelectFloatFromArray(Direct Array Index=$Particles.GroupID, Float Selection Array=DataInterface() {propertyValues {FloatData [4,4,20]}})
    )
    Use Linear Drag=true
    Use Rotational Drag=false
   )
   &SolveForcesAndVelocity(Clamp Velocity=false, Limit Acceleration=false)
  ]
  rendererClasses [&NiagaraSpriteRendererProperties]
  renderers [{rendererClass=&NiagaraSpriteRendererProperties}]
 }
 Sparks01 {
  simTarget=GPUComputeSim
  emitterUpdateScript [
   &EmitterState(Life Cycle Mode=::NewEnumerator1, Loop Behavior=::NewEnumerator1, Loop Delay=$User.Time Delay, Loop Duration=1, UseLoopDelay=true)
   &SpawnBurst_Instantaneous(Spawn Count=80, Use Spawn Probability=false, UseLoopCountLimit=false)
   &SpawnBurst_Instantaneous(Spawn Count=80, Spawn Group=1, Use Spawn Probability=false, UseLoopCountLimit=false) {module=SpawnBurst_Instantaneous001}
  ]
  particleSpawnScript [
   &InitializeParticle(
    Color Maximum=<rgba 2 0.2296 0.768 0.3739>
    Color Minimum=<rgba 0.6692 0.225 1.25 1>
    Color Mode=::NewEnumerator2
    Lifetime=&RandomRangeFloat(Minimum=0.5, Maximum=2)
    Sprite Rotation Mode=::NewEnumerator0
    Sprite Size Max=<3 20>
    Sprite Size Min=<1.5 10>
    Sprite Size Mode=::NewEnumerator2
    UsePositionOffset=false
    Write Material Random=true
   )
   /SetVariables(Particles.GroupID:NiagaraInt32=$Emitter.SpawnGroup) {isSetParameters=true}
   &ShapeLocation(
    Box Midpoint=<0 0.5 0.5>
    Box Size=&MakeVector(X=$User.DamageAreaLength, Y=100, Z=50)
    MaskSpawnGroup=false
    Offset=&SelectVectorFromArray(Direct Array Index=$Particles.GroupID, Vector Selection Array=DataInterface() {propertyValues {FloatData [<0 -200 0>, <0 200 0>]}})
    Offset Coordinate Space=::Local
    Offset Mode=::NewEnumerator0
    Override Seed=false
    Shape Origin=$Engine.Emitter.SimulationPosition
    Shape Primitive=::NewEnumerator2
    Surface Only=false
    UseOwnerScale=true
   )
   &AddVelocity(
    Cone Angle=20
    Cone Axis=&SelectVectorFromArray(Direct Array Index=$Particles.GroupID, Vector Selection Array=DataInterface() {propertyValues {FloatData [<0.5 -1 1>, <0.5 1 1>]}})
    Distribution Along Cone Axis=0
    Module.UseConeFalloff=false
    Module.UseVelDistribution=false
    Override Seed=false
    Speed Falloff From Cone Axis=0
    Velocity Mode=::NewEnumerator2
    Velocity Speed=&RandomRangeFloat(Minimum=750, Maximum=2500)
   ) {module=AddVelocity001}
  ]
  particleUpdateScript [
   /SetVariables(Particles.Glow03_Color=&Multiply_LinearColorByFloat(Linear Color=$Particles.Color, Float=&RandomRangeFloat(Minimum=0.25, Maximum=0.5))) {isSetParameters=true}
   /SetVariables(Particles.Glow03_Size=&Multiply_Vector2D_ByFloat(
     Float=&FloatFromCurve(
      FloatCurve=DataInterface() {propertyValues {Curve {keys [
          <at 0 0 RCIM_Cubic>
          <at 0.2 1.3 RCIM_Cubic>
          {arriveTangent=-0.1423, interpMode=RCIM_Cubic, leaveTangent=-0.1423, tangentMode=RCTM_User, time=1, value=0}
         ]}}}
      Scale Curve=&Multiply_Float(A=10, B=2)
     )
     Vector2D=<3 8>
    )) {isSetParameters=true}
   &ParticleState()
   &GravityForce(Gravity=&RandomRangeVector(Minimum=<0 0 -1000>, Maximum=<1 1 -2000>))
   &Drag(
    Drag=&FloatFromCurve(
     FloatCurve=DataInterface() {propertyValues {Curve {
        defaultValue=1
        keys [
         {tangentMode=RCTM_User, time=0, value=0.1}
         {arriveTangent=-1.2686, interpMode=RCIM_Cubic, leaveTangent=-1.2686, tangentMode=RCTM_User, time=0.2, value=0.5}
         {tangentMode=RCTM_User, time=1, value=0.1}
        ]
       }}}
     Scale Curve=3
    )
    Use Linear Drag=true
    Use Rotational Drag=false
   )
   &CurlNoiseForce(Noise Frequency=20, Noise Strength=350, Pan Noise=true, Pan Noise Field=<0 0 1>)
   &ScaleSpriteSize(Scale Factor=&Vector2DFromFloat(Value=&FloatFromCurve(
      FloatCurve=DataInterface() {propertyValues {Curve {keys [
          <at 0 0 RCIM_Cubic>
          {interpMode=RCIM_Cubic, tangentMode=RCTM_User, time=0.2, value=1.2}
          {interpMode=RCIM_Cubic, leaveTangent=-2.7132, time=1, value=0}
         ]}}}
      Scale Curve=&Multiply_Float(A=0.7, B=2)
     )))
   &SolveForcesAndVelocity(Clamp Velocity=false, Limit Acceleration=false)
   &ScaleSpriteSizeBySpeed(Max Scale Factor=<1 2.5>, Min Scale Factor=<0.85 0.5>, Velocity Threshold=3000)
  ]
  rendererClasses [&NiagaraSpriteRendererProperties]
  renderers [
   {rendererClass=&NiagaraSpriteRendererProperties}
   {rendererClass=&NiagaraSpriteRendererProperties, rendererIndex=1}
  ]
 }
 Wide_Trail {
  emitterUpdateScript [
   &EmitterState(
    Inactive Response=::NewEnumerator1
    Life Cycle Mode=::NewEnumerator1
    Loop Behavior=::NewEnumerator1
    Loop Delay=$User.Time Delay
    Loop Duration=1
    UseLoopDelay=true
   )
   &SpawnBurst_Instantaneous(Spawn Count=1, Use Spawn Probability=false)
  ]
  particleSpawnScript [
   &InitializeParticle(
    Color=<rgba 2.8194 4.3044 5 0.5>
    Lifetime=0.5
    Mass Mode=::NewEnumerator1
    Mesh Scale=&MakeVector(X=100, Y=30, Z=50)
    Mesh Scale Mode=::NewEnumerator4
    Position=&SimulationPosition()
    Position Mode=::NewEnumerator1
    Position Offset=<500 0 0>
    Sprite Size Mode=::NewEnumerator4
    UsePositionOffset=false
    Write Material Random=true
   ) {module=InitializeParticle001}
   &InitialMeshOrientation(
    Mesh Orientation Mode=::NewEnumerator2
    Orientation Coordinate Space=::Local
    Orientation Vector=<0 -1 0>
    Rotation=<0 0.25 0>
    Use Rotation Vector=true
   )
  ]
  particleUpdateScript [
   &ParticleState()
   &ScaleMeshSize(Scale Factor=&Multiply_VectorByFloat(
     Vector=<1 1 1>
     Float=&FloatFromCurve(FloatCurve=DataInterface() {propertyValues {Curve {keys [
          {arriveTangent=1.0119, interpMode=RCIM_Cubic, leaveTangent=1.0119, tangentMode=RCTM_User, time=0, value=0.3}
          {arriveTangent=0.0224, interpMode=RCIM_Cubic, leaveTangent=0.0224, tangentMode=RCTM_User, time=1, value=1}
         ]}}})
    ))
   &ScaleColor(
    Scale Alpha=&FloatFromCurve(FloatCurve=DataInterface() {propertyValues {Curve {keys [
         {arriveTangent=6.3626, interpMode=RCIM_Cubic, time=0, value=1}
         {arriveTangent=-2.2256, interpMode=RCIM_Cubic, leaveTangent=-2.2256, tangentMode=RCTM_User, time=1, value=0}
        ]}}})
    Scale RGB=&VectorFromCurve(VectorCurve=DataInterface() {propertyValues {XCurve {keys [<at 0 1 RCIM_Cubic>,<at 1 0>]}, YCurve {keys [<at 0 1 RCIM_Cubic>,<at 1 0>]}, ZCurve {keys [<at 0 1 RCIM_Cubic>,<at 1 0>]}}})
    ScaleA=true
    ScaleRGB=true
   )
   &SolveForcesAndVelocity(Clamp Velocity=false, Limit Acceleration=false)
   /SetVariables(Particles.Position=&Lerp_Vector(
     A=Linked() {linkedVariable {name=Particles.Initial.Position, type=NiagaraPosition}}
     B=&MakeVector(X=$User.DamageAreaLength, Y=0, Z=0)
     Alpha=&Power(a=$Particles.NormalizedAge, b=1.5)
    )) {isSetParameters=true}
  ]
  rendererClasses [&NiagaraMeshRendererProperties]
  renderers [
   {rendererClass=&NiagaraMeshRendererProperties, rendererIndex=0}
   {rendererClass=&NiagaraMeshRendererProperties, rendererIndex=1}
  ]
 }
 spikes001 {
  emitterUpdateScript [
   &EmitterState(Life Cycle Mode=::NewEnumerator1, Loop Behavior=::NewEnumerator1, Loop Delay=$User.Time Delay, Loop Duration=0.3, UseLoopDelay=true)
   &SpawnBurst_Instantaneous(
    Spawn Count=&RandomRangeInt(Minimum=5, Maximum=10)
    Use Spawn Probability=false
    UseLoopCountLimit=false
   )
   &SpawnBurst_Instantaneous(
    Spawn Count=&RandomRangeInt(Minimum=5, Maximum=10)
    Spawn Group=1
    Use Spawn Probability=false
    UseLoopCountLimit=false
   ) {module=SpawnBurst_Instantaneous001}
  ]
  particleSpawnScript [
   &InitializeParticle(
    Color=<rgba 8 0.9186 3.0722 0.3739>
    Lifetime Max=0.25
    Lifetime Min=0.175
    Lifetime Mode=::NewEnumerator1
    Sprite Size Max=<36.1112 150>
    Sprite Size Min=<50.7327 75>
    Sprite Size Mode=::NewEnumerator2
    UsePositionOffset=true
    Write Material Random=true
   )
   /SetVariables(Particles.GroupID:NiagaraInt32=$Emitter.SpawnGroup) {isSetParameters=true}
   &ShapeLocation(
    Box Midpoint=<0 0.5 0.5>
    Box Size=&MakeVector(X=$User.DamageAreaLength, Y=0, Z=0)
    MaskSpawnGroup=false
    Offset=&SelectVectorFromArray(Direct Array Index=$Particles.GroupID, Vector Selection Array=DataInterface() {propertyValues {FloatData [<0 -200 0>, <0 200 0>]}})
    Offset Coordinate Space=::Local
    Offset Mode=::NewEnumerator0
    Override Seed=false
    Shape Origin=$Engine.Emitter.SimulationPosition
    Shape Primitive=::NewEnumerator2
    Surface Only=false
    UseOwnerScale=true
   )
   &AddVelocity(
    Cone Angle=0
    Cone Axis=&SelectVectorFromArray(Direct Array Index=$Particles.GroupID, Vector Selection Array=DataInterface() {propertyValues {FloatData [<0.5 -1 0.5>, <0.5 1 0.5>]}})
    Distribution Along Cone Axis=0
    Module.UseConeFalloff=true
    Module.UseVelDistribution=true
    Override Seed=false
    Speed Falloff From Cone Axis=0
    Velocity Mode=::NewEnumerator2
    Velocity Speed=&RandomRangeFloat(Minimum=500, Maximum=1500)
   ) {module=AddVelocity001}
   &SubUVAnimation(End Frame=0, SubUV Animation Mode:ENiagara_SubUVLookupMode=::NewEnumerator2)
  ]
  particleUpdateScript [
   &ParticleState()
   &Drag(Drag=5, Use Linear Drag=true, Use Rotational Drag=false)
   &SolveForcesAndVelocity(Clamp Velocity=false, Limit Acceleration=false)
   &ScaleSpriteSize(
    Non-Uniform Curve Scale=&Multiply_Vector2D_ByFloat(Float=2.7, Vector2D=<0.9221 1.3>)
    Non-Uniform Curve Sprite Scale=DataInterface() {propertyValues {
      XCurve {keys [{tangentMode=RCTM_User, time=0, value=1}, {tangentMode=RCTM_User, time=1, value=0.8}]}
      YCurve {keys [{tangentMode=RCTM_User, time=0, value=0.5}, {tangentMode=RCTM_User, time=1, value=1.7}]}
     }}
    Scale Sprite Size Mode=::NewEnumerator3
   )
   &ScaleColor(
    Scale Alpha=&FloatFromCurve(FloatCurve=DataInterface() {propertyValues {Curve {keys [<at 0 0>,<at 0.5 1>,<at 1 0>]}}})
    Scale RGB=&Multiply_VectorByFloat(Vector=<1 1 1>, Float=1)
    ScaleA=true
    ScaleRGB=true
   )
   /SetVariables(Particles.Color dark:LinearColor=&MakeLinearColorFromVectorAndFloat(
     "Vector (RGB)"=&MakeVectorFromLinearColorRGB(Color=&MultiplyLinearColor(Linear Color=<rgba 0.12 0.0175 0.0584 10>, Scale Factor=&MakeVector4_FromFloat(Float=1)))
     "Float (Alpha)"=&Multiply_Float(
      A=2
      B=&FloatFromCurve(FloatCurve=DataInterface() {propertyValues {Curve {keys [<at 0 0>,<at 0.5 0.5574>,<at 1 0>]}}}, Scale Curve=2)
     )
    )) {isSetParameters=true}
   /SetVariables(Particles.Size02:Vector2f=&Multiply_Vector2D(
     A=&RandomRangeVector2D(Minimum=<38.6566 22.8626>, Maximum=<28.7057 71.9048>)
     B=&Vector2DFromCurve(
      Vector2Curve=DataInterface() {propertyValues {
        XCurve {keys [{tangentMode=RCTM_User, time=0, value=1}, {tangentMode=RCTM_User, time=1, value=0.8}]}
        YCurve {keys [{tangentMode=RCTM_User, time=0, value=0.5}, {tangentMode=RCTM_User, time=1, value=1.7}]}
       }}
      ScaleCurve=<5 5>
     )
    )) {isSetParameters=true}
  ]
  rendererClasses [&NiagaraSpriteRendererProperties]
  renderers [
   {rendererClass=&NiagaraSpriteRendererProperties}
   {rendererClass=&NiagaraSpriteRendererProperties, rendererIndex=1}
  ]
 }
}
summary {
 systemName=NS_Hit_BodySlam01
 emitters [
  {emitterName=Dirt01, rendererClasses [&NiagaraSpriteRendererProperties]}
  {emitterName=Mesh_Debris01, simTarget=GPUComputeSim, rendererClasses [&NiagaraMeshRendererProperties]}
  {emitterName=Smoke01, rendererClasses [&NiagaraSpriteRendererProperties]}
  {emitterName=spikes001, rendererClasses [&NiagaraSpriteRendererProperties]}
  {emitterName=Sparks01, simTarget=GPUComputeSim, rendererClasses [&NiagaraSpriteRendererProperties]}
  {emitterName=Wide_Trail, rendererClasses [&NiagaraMeshRendererProperties]}
 ]
 userVariables [
  User.DamageAreaLength {type=NiagaraFloat, defaultValue=1750, description=""}
  User.DamageAreaWidth {type=NiagaraFloat, defaultValue=500, description=""}
  User.HUE Brightness {type=NiagaraFloat, defaultValue=1.0324, description=""}
  User.HUE Color {type=NiagaraFloat, defaultValue=0.0853, description=""}
  User.HUE Saturation {type=NiagaraFloat, defaultValue=0.759, description=""}
  User.Time Delay {type=NiagaraFloat, defaultValue=0, description=""}
 ]
}
