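# Niagara telegraph effect for a body slam attack, marking a rectangular damage area sized by user length and width parameters.
# A looping decal and sprite mark the area, with a shared system WidthCurve driving decal width and smoke spread.
# Wide_Trail meshes sweep down the length while Energy mesh streaks and two-sided smoke add buildup.
# Normalized loop age ramps spawn rates and sizes, and shortens sine pulse periods so pulsing quickens.
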NiagaraSystem /TidalCrane_Swarm_Boss/Assets/FX/BodySlam/NS_Telegraph_BodySlam01
dependencies {
 usedDataInterfaces [
  &NiagaraDataInterfaceCurve
  &NiagaraDataInterfaceArrayFloat
  &NiagaraDataInterfaceVectorCurve
  &NiagaraDataInterfaceArrayFloat3
  &NiagaraDataInterfaceColorCurve
 ]
 usedDynamicInputs []
 usedModules [
  &self:SystemSpawnScript.SystemScriptSource.SystemScriptGraph.NiagaraNodeAssignment_0.SetVariables_#1
  &SystemState
 ]
}
emitters {
 Energy {
  emitterUpdateScript [
   &EmitterState()
   &SpawnRate(
    SpawnRate=&Lerp_Float(A=5, B=12, Alpha=&Power(a=$Emitter.NormalizedLoopAge, b=0.5))
    Use Spawn Probability=false
   )
   /SetVariables(Emitter.RefractionIntensity=10) {isSetParameters=true}
  ]
  particleSpawnScript [
   /SetVariables(
    Particles.ColorEnd:LinearColor=&MultiplyLinearColor(
     Linear Color=&RandomRangeLinearColor(Minimum=<rgba 0.8333 0.5548 0.4991 1>, Maximum=<rgba 1 0.5277 0.3125 1>)
     Scale Factor=&MakeFloat4FromVectorAndFloat("Vector (RGB)"=&VectorFromFloat(Value=10), "Float (Alpha)"=1)
    )
    Particles.ColorStart:LinearColor=&MultiplyLinearColor(
     Linear Color=&RandomRangeLinearColor(Minimum=<rgba 0.9688 0.5981 0.4856 1>, Maximum=<rgba 0.6562 0.4477 0.3526 1>)
     Scale Factor=&MakeFloat4FromVectorAndFloat("Vector (RGB)"=&VectorFromFloat(Value=1), "Float (Alpha)"=1)
    )
   ) {isSetParameters=true}
   &InitializeParticle(
    Color=$Particles.ColorStart
    Lifetime Max=0.8
    Lifetime Min=0.4
    Lifetime Mode=::NewEnumerator1
    Mesh Scale Max=<7 1.4 1>
    Mesh Scale Min=<5 1 1>
    Mesh Scale Mode=::NewEnumerator2
    Position Offset=<0 0 100>
    UsePositionOffset=true
    Write Material Random=true
   )
   &InitialMeshOrientation(Mesh Axis To Orient=<-1 0 0>, Mesh Orientation Mode=::NewEnumerator2, Orientation Coordinate Space=::Local, Use Rotation Vector=true)
  ]
  particleUpdateScript [
   /SetVariables(Particles.vecYZ:NiagaraFloat=&FloatFromCurve(
     FloatCurve=DataInterface() {propertyValues {Curve {keys [{interpMode=RCIM_Cubic, tangentMode=RCTM_Break, time=0, value=1}, {interpMode=RCIM_Cubic, tangentMode=RCTM_Break, time=1, value=2}]}}}
     Scale Curve=&FloatFromCurve(
      FloatCurve=DataInterface() {propertyValues {Curve {keys [
          {interpMode=RCIM_Cubic, tangentMode=RCTM_Break, time=0, value=0.3}
          {arriveTangent=2.0952, interpMode=RCIM_Cubic, tangentMode=RCTM_Break, time=0.35, value=1}
          {arriveTangent=1.831, interpMode=RCIM_Cubic, leaveTangent=-2.7219, tangentMode=RCTM_Break, time=1, value=1.5}
         ]}}}
      CurveIndex=$Emitter.NormalizedLoopAge
     )
    )) {isSetParameters=true}
   &ParticleState()
   &DynamicMaterialParameters(
    Index 0 Param 1=&FloatFromCurve(FloatCurve=DataInterface() {propertyValues {Curve {keys [
         {arriveTangent=0.0883, interpMode=RCIM_Cubic, leaveTangent=0.0883, tangentMode=RCTM_User, time=0, value=-0.1}
         {arriveTangent=0.0511, interpMode=RCIM_Cubic, tangentMode=RCTM_Break, time=1, value=-0.8}
        ]}}})
    Index 0 Param 2=&FloatFromCurve(FloatCurve=DataInterface() {propertyValues {Curve {keys [{arriveTangent=-0.5818, interpMode=RCIM_Cubic, leaveTangent=-0.5818, tangentMode=RCTM_User, time=0, value=1}, <at 1 0.9>]}}})
    Index 0 Param 3=&FloatFromCurve(FloatCurve=DataInterface() {propertyValues {Curve {keys [<at 0 0>,<at 0.15 1>,<at 1 1.2>]}}}, Scale Curve=0.7)
    Index 1 Param 1=1
    Index 1 Param 2=1
    Index 1 Param 3=1
    Index 1 Param 4=1
    Index 2 Param 1=1
    Index 2 Param 2=1
    Index 2 Param 3=&FloatFromCurve(
     FloatCurve=DataInterface() {propertyValues {Curve {keys [
         {interpMode=RCIM_Cubic, leaveTangent=0.1609, tangentMode=RCTM_Break, time=0, value=1}
         {arriveTangent=0.0839, interpMode=RCIM_Cubic, tangentMode=RCTM_Break, time=1, value=0.1}
        ]}}}
     Scale Curve=20
    )
    Index 3 Param 1=&RandomRangeFloat(Minimum=0, Maximum=1)
    Index 3 Param 3=1
    Index 3 Param 4=1
    Module.Param0WriteEnabled=true
    Module.Param0WriteEnabled1=true
    Module.Param0WriteEnabled2=true
    Module.Param0WriteEnabled3=true
    Module.Param1WriteEnabled=true
    Module.Param1WriteEnabled1=true
    Module.Param1WriteEnabled2=true
    Module.Param1WriteEnabled3=true
    Module.Param2WriteEnabled=true
    Module.Param2WriteEnabled1=true
    Module.Param2WriteEnabled2=true
    Module.Param2WriteEnabled3=true
    Module.Param3WriteEnabled=false
    Module.Param3WriteEnabled1=true
    Module.Param3WriteEnabled2=true
    Module.Param3WriteEnabled3=true
    Write Parameter Index 1=true
    Write Parameter Index 2=true
    Write Parameter Index 3=true
   )
   &ScaleColor(
    Scale Alpha=&FloatFromCurve(
     FloatCurve=DataInterface() {propertyValues {Curve {keys [
         {interpMode=RCIM_Cubic, tangentMode=RCTM_Break, time=0, value=0}
         {interpMode=RCIM_Cubic, tangentMode=RCTM_Break, time=0.3, value=1}
         {interpMode=RCIM_Cubic, tangentMode=RCTM_Break, time=1, value=0}
        ]}}}
     Scale Curve=&FloatFromCurve(
      FloatCurve=DataInterface() {propertyValues {Curve {keys [
          {interpMode=RCIM_Cubic, leaveTangent=0.3798, tangentMode=RCTM_Break, time=0, value=0.2}
          {interpMode=RCIM_Cubic, tangentMode=RCTM_Break, time=1.6, value=1}
          {arriveTangent=-2.2529, interpMode=RCIM_Cubic, tangentMode=RCTM_Break, time=2.2, value=0}
         ]}}}
      CurveIndex=$Emitter.NormalizedLoopAge
     )
    )
    Scale RGB=&MakeVectorFromLinearColorRGB(Color=&LerpLinearColors(
      StartColor=$Particles.ColorStart
      EndColor=$Particles.ColorEnd
      LerpFactor=&FloatFromCurve(FloatCurve=DataInterface() {propertyValues {Curve {keys [
           {interpMode=RCIM_Cubic, leaveTangent=2.2328, tangentMode=RCTM_Break, time=0, value=0}
           {interpMode=RCIM_Cubic, tangentMode=RCTM_Break, time=1, value=1}
          ]}}})
     ))
    ScaleA=true
    ScaleRGB=true
   )
   &ScaleMeshSize(Scale Factor=&MakeVector(
     X=&FloatFromCurve(
      FloatCurve=DataInterface() {propertyValues {Curve {keys [
          {interpMode=RCIM_Cubic, leaveTangent=-4.4441, tangentMode=RCTM_Break, time=0, value=2}
          {interpMode=RCIM_Cubic, tangentMode=RCTM_Break, time=1, value=0.2}
         ]}}}
      Scale Curve=&FloatFromCurve(
       FloatCurve=DataInterface() {propertyValues {Curve {keys [
           {interpMode=RCIM_Cubic, tangentMode=RCTM_Break, time=0, value=0.4}
           {arriveTangent=1.4904, interpMode=RCIM_Cubic, leaveTangent=0.4062, tangentMode=RCTM_Break, time=0.75, value=1}
           {interpMode=RCIM_Cubic, tangentMode=RCTM_Break, time=1.6, value=1.2}
           {interpMode=RCIM_Cubic, tangentMode=RCTM_Break, time=2.2, value=0.6}
          ]}}}
       CurveIndex=$Emitter.NormalizedLoopAge
       Scale Curve=1.55
      )
     )
     Y=$Particles.vecYZ
     Z=$Particles.vecYZ
    ))
  ]
  rendererClasses [&NiagaraMeshRendererProperties]
  renderers [{rendererClass=&NiagaraMeshRendererProperties, rendererIndex=0}]
 }
 SingleLoopingParticle {
  emitterUpdateScript [
   &EmitterState()
   &SpawnBurst_Instantaneous(Use Spawn Probability=false, UseLoopCountLimit=false)
   /SetVariables(
    Emitter.Brightness:NiagaraFloat=&Sine(
     Normalized Angle=$Emitter.Age
     Period=&Lerp_Float(A=1, B=0.35, Alpha=$Emitter.NormalizedLoopAge)
     Scale=0.5
     Bias=1
    )
    Emitter.OuterEdgeRadius:NiagaraFloat=&Sine(
     Normalized Angle=&Add_Float(A=$Emitter.Age, B=0.1)
     Period=&Lerp_Float(A=1, B=0.35, Alpha=$Emitter.NormalizedLoopAge)
     Scale=0.5
     Bias=1
    )
    Emitter.Radius:NiagaraFloat=&FloatFromCurve(
     FloatCurve=DataInterface() {propertyValues {Curve {keys [<at 0 0.005 RCIM_Cubic>, {arriveTangent=1.3142, tangentMode=RCTM_User, time=0.95, value=1}, <at 1 0>]}}}
     CurveIndex=$Emitter.NormalizedLoopAge
    )
   ) {isSetParameters=true}
  ]
  particleSpawnScript [
   &InitializeParticle(
    Color=<rgba 0.9798 0.0017 1 1>
    Lifetime=$Emitter.CurrentLoopDuration
    Position Offset=&MakeVector(X=&Multiply_Float(A=$User.DamageAreaLength, B=0.5), Y=0, Z=0)
    Sprite Size=<1500 1250>
    Sprite Size Mode=::NewEnumerator4
    UsePositionOffset=true
    Write Material Random=true
   )
   &SpriteFacingAndAlignment(Sprite Alignment=<0 -1 0>, Sprite Facing=<0 0 1>, WriteAlignment=true, WriteFacing=true) {enabled=false}
  ]
  particleUpdateScript [
   &ParticleState()
   /Niagara/Modules/Decal/Decal_Attributes(
    Alignment Coordinate Space:ENiagaraCoordinateSpace=::Local
    Decal Alignment:Vector3f=<0 -1 0>
    Decal Facing Mode=1
    Decal Size:Vector3f=&MakeVector(
     X=200
     Y=&FloatFromCurve(FloatCurve=$System.WidthCurve, Scale Curve=$User.DamageAreaWidth)
     Z=$User.DamageAreaLength
    )
    Facing Coordinate Space:ENiagaraCoordinateSpace=::World
    Orientation=1
    Set Alignment=true
    Size=1
    Visible=1
   )
   &SolveForcesAndVelocity(Clamp Velocity=false, Limit Acceleration=false)
  ]
  rendererClasses [&NiagaraSpriteRendererProperties, &NiagaraDecalRendererProperties]
  renderers [
   {rendererClass=&NiagaraSpriteRendererProperties}
   {rendererClass=&NiagaraDecalRendererProperties, rendererIndex=1}
  ]
 }
 Smoke01 {
  emitterUpdateScript [
   &EmitterState()
   &SpawnRate(
    SpawnRate=&FloatFromCurve(
     FloatCurve=DataInterface() {propertyValues {Curve {keys [{arriveTangent=2.4165, interpMode=RCIM_Cubic, leaveTangent=2.4165, tangentMode=RCTM_User, time=0.01, value=0}, <at 0.65 1>]}}}
     CurveIndex=$Emitter.NormalizedLoopAge
     Scale Curve=/Niagara/DynamicInputs/Integer/V2/Round_Float(Float:NiagaraFloat=&Sine(
       Normalized Angle=$Emitter.Age
       Period=&Lerp_Float(A=1, B=0.2, Alpha=$Emitter.NormalizedLoopAge)
       Scale=100
       Bias=50
      ))
    )
    Use Spawn Probability=false
   )
  ]
  particleSpawnScript [
   &InitializeParticle(
    Lifetime=&Add_Float(
     A=0.25
     B=&FloatFromCurve(
      FloatCurve=$System.WidthCurve
      CurveIndex=$Emitter.NormalizedLoopAge
      Scale Curve=&RandomRangeFloat(Minimum=0.55, Maximum=1)
     )
    )
    Mass Mode=::NewEnumerator3
    Sprite Rotation Mode=::NewEnumerator2
    Sprite Size Mode=::NewEnumerator1
    Uniform Sprite Size Max=350
    Uniform Sprite Size Min=150
    UsePositionOffset=false
    Write Material Random=true
   )
   &CameraOffset(Camera Offset Amount=70)
   &PartitionParticles()
   /SetVariables(Particles.GroupID:NiagaraInt32=$Particles.PartitionParticles.Partition) {isSetParameters=true}
   &ShapeLocation(
    Box Midpoint=<0 0.5 0>
    Box Size=&MakeVector(X=$User.DamageAreaLength, Y=100, Z=50)
    MaskSpawnGroup=false
    Offset=&Multiply_VectorByFloat(
     Vector=&SelectVectorFromArray(Direct Array Index=$Particles.GroupID, Vector Selection Array=DataInterface() {propertyValues {FloatData [<0 -200 0>, <0 200 0>]}})
     Float=&FloatFromCurve(FloatCurve=$System.WidthCurve, CurveIndex=$Emitter.NormalizedLoopAge)
    )
    Offset Coordinate Space=::Local
    Offset Mode=::NewEnumerator0
    Override Seed=false
    Shape Origin=$Engine.Emitter.SimulationPosition
    Shape Primitive=::NewEnumerator2
    Surface Only=false
    UseOwnerScale=true
   )
   &AddVelocity(
    Cone Angle=30
    Cone Axis=&SelectVectorFromArray(Direct Array Index=$Particles.GroupID, Vector Selection Array=DataInterface() {propertyValues {FloatData [<1 -0.75 0.5>, <1 0.75 0.5>]}})
    Distribution Along Cone Axis=0
    Module.UseConeFalloff=false
    Module.UseVelDistribution=false
    Override Seed=false
    Speed Falloff From Cone Axis=150
    Velocity Mode=::NewEnumerator2
    Velocity Speed=&FloatFromCurve(
     FloatCurve=$System.WidthCurve
     CurveIndex=$Emitter.NormalizedLoopAge
     Scale Curve=&RandomRangeFloat(Minimum=350, Maximum=750)
    )
   ) {module=AddVelocity001}
  ]
  particleUpdateScript [
   &ParticleState()
   &ScaleSpriteSize(
    Uniform Curve Scale=&Lerp_Float(A=0.5, B=1, Alpha=$Emitter.NormalizedLoopAge)
    Uniform Curve Sprite Scale=DataInterface() {propertyValues {Curve {keys [
        {arriveTangent=1.5303, interpMode=RCIM_Cubic, leaveTangent=1.5303, tangentMode=RCTM_User, time=0, value=0.35}
        {tangentMode=RCTM_User, time=1, value=1}
       ]}}}
   )
   /SetVariables(Particles.SubImageIndex=&FloatFromCurve(FloatCurve=DataInterface() {propertyValues {Curve {keys [
         {arriveTangent=29.8623, interpMode=RCIM_Cubic, leaveTangent=29.8623, tangentMode=RCTM_User, time=-0.0046, value=5.0556}
         <at 1.0011 22.1689 RCIM_Cubic>
        ]}}})) {isSetParameters=true}
   &DynamicMaterialParameters(
    Index 0 Param 1=&FloatFromCurve(FloatCurve=DataInterface() {propertyValues {Curve {keys [<at -0.0046 1.0113 RCIM_Cubic>,<at 1.0011 0.002>]}}})
    Index 0 Param 4=&FloatFromCurve(FloatCurve=DataInterface() {propertyValues {Curve {keys [<at 0 0 RCIM_Cubic>,<at 1 0.5>]}}})
    Module.Param0WriteEnabled=true
    Module.Param1WriteEnabled=false
    Module.Param2WriteEnabled=false
    Module.Param3WriteEnabled=true
   )
   &Color(Color=&RandomRangeLinearColor(
     Minimum=&ColorFromCurve(ColorCurve=DataInterface() {propertyValues {
        AlphaCurve {keys [<at 0.4505 0.15 RCIM_Cubic>,<at 1 0 RCIM_Cubic>]}
        BlueCurve {keys [<at 0 0.2025 RCIM_Cubic>,<at 1 0.2025 RCIM_Cubic>]}
        GreenCurve {keys [<at 0 0.3859 RCIM_Cubic>,<at 1 0.3858 RCIM_Cubic>]}
        RedCurve {keys [<at 0 0.5972 RCIM_Cubic>,<at 1 0.5972 RCIM_Cubic>]}
       }})
     Maximum=&ColorFromCurve(ColorCurve=DataInterface() {propertyValues {
        AlphaCurve {keys [<at 0.4743 0.1 RCIM_Cubic>,<at 1 0 RCIM_Cubic>]}
        BlueCurve {keys [<at 0 0.0508 RCIM_Cubic>,<at 1 0.0508 RCIM_Cubic>]}
        GreenCurve {keys [<at 0 0.0974 RCIM_Cubic>,<at 1 0.0974 RCIM_Cubic>]}
        RedCurve {keys [<at 0 0.15 RCIM_Cubic>,<at 1 0.15 RCIM_Cubic>]}
       }})
    ))
   &GravityForce(Gravity=&RandomRangeVector(Minimum=<0 0 -250>, Maximum=<1 1 -550>))
   &Drag(
    Drag=&FloatFromCurve(
     FloatCurve=DataInterface() {propertyValues {Curve {defaultValue=1, keys [{tangentMode=RCTM_User, time=0, value=0}, <at 0.3 0.5>, {tangentMode=RCTM_User, time=1, value=0}]}}}
     Scale Curve=&SelectFloatFromArray(Direct Array Index=$Particles.GroupID, Float Selection Array=DataInterface() {propertyValues {FloatData [4,4,20]}})
    )
    Use Linear Drag=true
    Use Rotational Drag=false
   )
   &SolveForcesAndVelocity(Clamp Velocity=false, Limit Acceleration=false)
  ]
  rendererClasses [&NiagaraSpriteRendererProperties]
  renderers [{rendererClass=&NiagaraSpriteRendererProperties}]
 }
 Wide_Trail {
  emitterUpdateScript [
   &EmitterState()
   &SpawnBurst_Instantaneous(Use Spawn Probability=false, UseLoopCountLimit=false)
   &SpawnRate(
    SpawnRate=&FloatFromCurve(
     FloatCurve=DataInterface() {propertyValues {Curve {keys [<at 0 0.2>,<at 0.9 1>,<at 0.95 0>]}}}
     CurveIndex=$Emitter.NormalizedLoopAge
     Scale Curve=5
    )
    Use Spawn Probability=false
   )
  ]
  particleSpawnScript [
   &InitializeParticle(
    Color=&LerpLinearColors(StartColor=<rgba 0.5 0.4147 0.3376 0.5>, EndColor=<rgba 1 0.9081 0.7448 0.5>, LerpFactor=$Emitter.NormalizedLoopAge)
    Lifetime=0.5
    Mass Mode=::NewEnumerator1
    Mesh Scale=&MakeVector(
     X=&Lerp_Float(A=50, B=100, Alpha=$Emitter.NormalizedLoopAge)
     Y=30
     Z=&Multiply_Float(
      A=&Divide_Float(A=$User.DamageAreaWidth, B=20)
      B=&Lerp_Float(A=0.3, B=1.5, Alpha=&Power(a=$Emitter.NormalizedLoopAge, b=0.5))
     )
    )
    Mesh Scale Mode=::NewEnumerator4
    Position=&SimulationPosition()
    Position Mode=::NewEnumerator1
    Sprite Size Mode=::NewEnumerator4
    UsePositionOffset=false
    Write Material Random=true
   ) {module=InitializeParticle001}
   &InitialMeshOrientation(
    Mesh Orientation Mode=::NewEnumerator2
    Orientation Coordinate Space=::Local
    Orientation Vector=<0 -1 0>
    Rotation=<0 0.25 0>
    Use Rotation Vector=true
   )
  ]
  particleUpdateScript [
   &ParticleState()
   &ScaleMeshSize(Scale Factor=&Multiply_VectorByFloat(
     Vector=<1 1 1>
     Float=&FloatFromCurve(
      FloatCurve=DataInterface() {propertyValues {Curve {keys [
          {arriveTangent=1.0119, interpMode=RCIM_Cubic, leaveTangent=1.0119, tangentMode=RCTM_User, time=0, value=0.3}
          {arriveTangent=0.0224, interpMode=RCIM_Cubic, leaveTangent=0.0224, tangentMode=RCTM_User, time=1, value=1}
         ]}}}
      Scale Curve=&SelectFloatFromArray(
       Direct Array Index=&SpawnGroup()
       Float Selection Array=DataInterface() {propertyValues {FloatData [1.1,1.35,1.75,2]}}
      )
     )
    ))
   &ScaleColor(
    Scale Alpha=&FloatFromCurve(
     FloatCurve=DataInterface() {propertyValues {Curve {keys [
         {arriveTangent=6.3626, interpMode=RCIM_Cubic, time=0, value=1}
         {arriveTangent=-2.2256, interpMode=RCIM_Cubic, leaveTangent=-2.2256, tangentMode=RCTM_User, time=1, value=0}
        ]}}}
     Scale Curve=0.85
    )
    Scale RGB=&VectorFromCurve(VectorCurve=DataInterface() {propertyValues {XCurve {keys [<at 0 1 RCIM_Cubic>,<at 1 0>]}, YCurve {keys [<at 0 1 RCIM_Cubic>,<at 1 0>]}, ZCurve {keys [<at 0 1 RCIM_Cubic>,<at 1 0>]}}})
    ScaleA=true
    ScaleRGB=true
   )
   &SolveForcesAndVelocity(Clamp Velocity=false, Limit Acceleration=false)
   /SetVariables(Particles.Position=&Lerp_Vector(
     A=Linked() {linkedVariable {name=Particles.Initial.Position, type=NiagaraPosition}}
     B=&MakeVector(X=$User.DamageAreaLength, Y=0, Z=0)
     Alpha=&Power(a=$Particles.NormalizedAge, b=1.35)
    )) {isSetParameters=true}
  ]
  rendererClasses [&NiagaraMeshRendererProperties]
  renderers [
   {rendererClass=&NiagaraMeshRendererProperties, rendererIndex=0}
   {rendererClass=&NiagaraMeshRendererProperties, rendererIndex=1}
  ]
 }
}
summary {
 systemName=NS_Telegraph_BodySlam01
 emitters [
  {
   emitterName=SingleLoopingParticle
   rendererClasses [&NiagaraSpriteRendererProperties, &NiagaraDecalRendererProperties]
  }
  {emitterName=Wide_Trail, rendererClasses [&NiagaraMeshRendererProperties]}
  {emitterName=Energy, rendererClasses [&NiagaraMeshRendererProperties]}
  {emitterName=Smoke01, rendererClasses [&NiagaraSpriteRendererProperties]}
 ]
 userVariables [
  User.DamageAreaLength {type=NiagaraFloat, defaultValue=1700, description=""}
  User.DamageAreaWidth {type=NiagaraFloat, defaultValue=500, description=""}
  User.HUE Brightness {type=NiagaraFloat, defaultValue=1, description=""}
  User.HUE Color {type=NiagaraFloat, defaultValue=0, description=""}
  User.HUE Saturation {type=NiagaraFloat, defaultValue=1, description=""}
  User.SpriteColor {type=LinearColor, description="", defaultValue {struct=LinearColor, value=<rgba 1 1 1 1>}}
  User.Stretching Mesh by X {type=NiagaraFloat, defaultValue=10, description=""}
  User.Time Delay {type=NiagaraFloat, defaultValue=0, description=""}
 ]
}
