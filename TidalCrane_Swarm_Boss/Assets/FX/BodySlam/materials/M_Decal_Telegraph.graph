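# Translucent deferred decal material that, by its name and path, telegraphs a boss body slam as a cracked, warm-glowing ground strip.
# Cloud noise distorts UVs for sine and power edge masks, which shape world-projected cracks and a bright outline controlled by Brightness.
# Exponential radial gradients and U-axis falloffs fade the strip.
# The masks feed a warm emissive tint via MF_FX_EmissiveAdjustments and opacity-style blends, alongside a brown color gradient.
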Material /TidalCrane_Swarm_Boss/Assets/FX/BodySlam/materials/M_Decal_Telegraph
graph {
 properties {
  bAutomaticallySetUsageInEditor=false
  bUsedWithNiagaraSprites=true
  blendMode=BLEND_Translucent
  materialDomain=MD_DeferredDecal
  usedShadingModels=Default Lit
 }
 nodes [
  n1 = MaterialFunctionCall(
   UVs=Multiply(A=TextureCoordinate(), B=Constant2Vector() {g=1, r=0.5})
   CenterPosition=Constant2Vector() {g=0.5}
   Radius=ScalarParameter() {defaultValue=0.5, parameterName=RadialRadius}
   Density=Constant() {r=5}
  ) {materialFunction=&RadialGradientExponential}
  n2 = TextureCoordinate()
  n3 = Power(Base=ComponentMask(None=@n2) {r=true}) {constExponent=1}
  n4 = Sine(None=Subtract(A=ComponentMask(None=@n2) {g=true}) {constB=0.25})
  n5 = Multiply(A=@n1.RadialGradientExponential, B=Multiply(A=OneMinus(None=@n3), B=Saturate(None=@n4)))
  n6 = Constant() {r=3}
  n7 = Constant() {r=1}
  n8 = TextureSample(UVs=Multiply(
    A=TextureCoordinate()
    B=MaterialFunctionCall(X=@n6, Y=@n7) {materialFunction=&MakeFloat2}.Result
   )) {texture=/TidalCrane_Swarm_Boss/Assets/FX/BodySlam/textures/T_Cracks01}
  n9 = Sine(None=@n1.RadialGradientExponential)
  n10 = Multiply(A=WorldPosition().XY) {constB=0.0035}
  n11 = TextureObject() {texture=/Game/Weapons/FORT_Melee/Pickaxe_Buffet_Female/FX/T_Cracks}
  n12 = TextureCoordinate() {uTiling=5}
  n13 = ComponentMask(None=ConstantBiasScale(None=TextureSample(UVs=@n12) {samplerType=SAMPLERTYPE_LinearGrayscale, texture=/Game/Effects/Fort_Effects/Emotes/UltraLight/Texture/T_Cloud_Greyscale}.RGB) {bias=-0.5, scale=2}) {g=true, r=true}
  n14 = ComponentMask(None=Add(A=TextureCoordinate(), B=Multiply(A=@n13) {constB=0.075})) {g=true}
  n15 = Constant() {r=5}
  n16 = Constant() {r=20}
  n17 = Add(A=Power(Base=@n14, Exp=@n15) {constExponent=15}, B=Power(Base=@n14, Exp=@n16) {constExponent=75})
  n18 = OneMinus(None=@n14)
  n19 = Add(A=Power(Base=@n18, Exp=@n15) {constExponent=15}, B=Power(Base=@n18, Exp=@n16) {constExponent=75})
  n20 = Add(A=Multiply(A=@n13) {constB=0.05}, B=TextureCoordinate())
  n21 = ComponentMask(None=Sine(None=Add(A=@n20) {constB=-0.035}) {period=1.85}) {g=true}
  n22 = Saturate(None=Multiply(A=@n21) {constB=20})
  n23 = Multiply(
   A=Multiply(A=TextureSample(UVs=@n10, Tex=@n11) {texture=/Game/Effects/Fort_Effects/Textures/T_Cracks_Gradient}.RGB, B=Add(A=@n17, B=@n19))
   B=@n22
  )
  n24 = Multiply(A=Multiply(A=Saturate(None=@n9)) {constB=2}, B=@n23)
  n25 = Multiply(A=@n1.RadialGradientExponential, B=@n23)
  n26 = ScalarParameter() {defaultValue=5, parameterName=Brightness}
  n27 = OneMinus(None=Saturate(None=Abs(None=@n21)))
  n28 = TextureCoordinate()
  n29 = Constant2Vector() {g=1, r=0.5}
  n30 = Multiply(
   A=Saturate(None=Power(Base=@n27) {constExponent=35})
   B=MaterialFunctionCall(
    UVs=Multiply(A=@n28, B=@n29)
    CenterPosition=Constant2Vector() {g=0.5, r=0.1}
    Radius=ScalarParameter() {defaultValue=1, parameterName=OuterEdgeRadialRadius}
    Density=ScalarParameter() {defaultValue=5, parameterName=RadialDensity}
   ) {materialFunction=&RadialGradientExponential}.RadialGradientExponential
  )
  n31 = Multiply(A=Add(A=Add(A=@n24, B=@n25), B=Multiply(A=@n26, B=@n30))) {constB=2}
  Add(A=Power(Base=Multiply(A=@n5, B=@n8.B)) {constExponent=1.25}, B=@n31)
  LinearInterpolate(
   A=Constant3Vector() {constant=<rgba 0.25 0.1265 0.1185 1>}
   B=Constant3Vector() {constant=<rgba 0.0312 0.007 0.0019 1>}
   Alpha=ComponentMask(None=TextureCoordinate()) {r=true}
  )
  MaterialFunctionCall(Emissive=Multiply(A=Constant3Vector() {constant=<rgba 0.3438 0.1802 0.0483 1>}, B=@n31)) {materialFunction=&MF_FX_EmissiveAdjustments}
  Multiply(A=@n5, B=@n8.R)
  n32 = Constant() {r=0.55}
  n33 = ComponentMask(None=@n20) {r=true}
  Multiply(A=Add(A=@n31, B=Multiply(A=@n22, B=@n32)) {constB=0.35}, B=Power(Base=OneMinus(None=@n33)) {constExponent=0.2})
 ]
}
