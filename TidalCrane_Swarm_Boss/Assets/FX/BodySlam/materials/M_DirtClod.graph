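# M_DirtClod is a masked, lit Niagara sprite material for dirt clod particles in the Tidal Crane Swarm boss's body slam effect.
# The clod texture's alpha drives a BumpOffset that shifts the diffuse and normal samples for parallax depth.
# Diffuse RGB blends between AO dark and light parameters, tinted by particle color.
# Particle alpha feeds DitherTemporalAA, whose inverted result is subtracted from texture alpha for a dithered fade.
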Material /TidalCrane_Swarm_Boss/Assets/FX/BodySlam/materials/M_DirtClod
graph {
 properties {bUsedWithNiagaraSprites=true, blendMode=BLEND_Masked, pixelDepthOffsetMode=PDOM_Legacy, usedShadingModels=Default Lit}
 nodes [
  n1 = ParticleColor()
  n2 = BumpOffset(Height=TextureSample() {texture=/Game/Effects/Fort_Effects/Textures/Dirt/T-DirtClodCluster_D}.A)
  n3 = TextureSample(UVs=@n2) {texture=&T-DirtClodCluster_D}
  Multiply(
   A=@n1.RGB
   B=LinearInterpolate(A=ScalarParameter() {parameterName=AO Dark}, B=ScalarParameter() {defaultValue=1, parameterName=AO Light}, Alpha=@n3.RGB)
  )
  ScalarParameter() {defaultValue=0.9, parameterName=Roughness}
  Subtract(
   A=@n3.A
   B=OneMinus(None=MaterialFunctionCall(Alpha Threshold=@n1.A) {materialFunction=&DitherTemporalAA}.Result)
  )
  TextureSample(UVs=@n2) {samplerType=SAMPLERTYPE_Normal, texture=/Game/Effects/Fort_Effects/Textures/Dirt/T-DirtClodCluster_N}
 ]
}
